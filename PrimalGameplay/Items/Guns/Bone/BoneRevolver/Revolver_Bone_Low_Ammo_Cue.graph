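# Sound cue for the low-ammo warning of the Bone Revolver gun.
# A single wave player reuses a flintlock revolver low-ammo sample, fed through a modulator that pitches it down within a narrow random range.
# It uses the ranged weapon low-ammo degraded sound class at reduced volume, and its own concurrency override blocks new instances rather than cutting off playing ones.
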SoundCue /PrimalGameplay/Items/Guns/Bone/BoneRevolver/Revolver_Bone_Low_Ammo_Cue
root=Modulator_0
nodes [
 {id=Modulator_0, children [WavePlayer_0], properties {pitchMax=0.85, pitchMin=0.8, volumeMax=1, volumeMin=1}}
 {id=WavePlayer_0, properties {soundWaveAssetPtr=/Game/Sounds/Fort_Weapons/FlintLock/Flintlock_Revolver_LowAmmo_02}}
]
properties {
 bOverrideConcurrency=true
 maxDistance=2097152
 soundClassObject=/Game/Sounds/FortSoundClassesAndModes/Fort_WeaponRanged_LowAmmo_Degraded_PSM
 volumeMultiplier=0.7
 concurrencyOverrides {resolutionRule=PreventNew}
}
