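# Sound cue for picking up the bone revolver weapon.
# A single wave player feeds the Revolver_Bone_Pickup_01 wave straight to the root, with no randomization or mixing.
# It plays in the loot-collect sound class with weapon pickup attenuation and reduced volume.
# A concurrency override allows one instance at a time and blocks new plays while one is active.
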SoundCue /PrimalGameplay/Items/Guns/Bone/BoneRevolver/Revolver_Bone_Pickup_Cue
root=WavePlayer_4
nodes [{id=WavePlayer_4, properties {soundWaveAssetPtr=&/Revolver_Bone_Pickup_01}}]
properties {
 attenuationSettings=&Fort_Weapon_Pickup_Att
 bOverrideConcurrency=true
 maxDistance=3500
 soundClassObject=&Fort_Loot_Collect
 volumeMultiplier=0.398
 concurrencyOverrides {maxCount=1, resolutionRule=PreventNew}
}
