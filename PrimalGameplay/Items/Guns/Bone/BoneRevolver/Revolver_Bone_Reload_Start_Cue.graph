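# Sound cue for the start of the Bone Revolver's reload, played as weapon foley.
# A single wave player feeds a modulator that randomizes pitch for slight variation between plays.
# It uses the shared weapon reload attenuation settings at reduced volume, with overridden concurrency that prevents new instances once the limit is reached.
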SoundCue /PrimalGameplay/Items/Guns/Bone/BoneRevolver/Revolver_Bone_Reload_Start_Cue
root=Modulator_2
nodes [
 {id=Modulator_2, children [WavePlayer_0], properties {pitchMax=1.2, pitchMin=0.9}}
 {id=WavePlayer_0, properties {soundWaveAssetPtr=&/Revolver_Bone_Reload_Start_01}}
]
properties {
 attenuationSettings=&Fort_Weapon_Reload_Att
 bOverrideConcurrency=true
 maxDistance=2250
 soundClassObject=&Fort_WeaponFoley
 volumeMultiplier=0.401
 attenuationOverrides {falloffDistance=4100, attenuationShapeExtents {x=900}}
 concurrencyOverrides {resolutionRule=PreventNew}
}
